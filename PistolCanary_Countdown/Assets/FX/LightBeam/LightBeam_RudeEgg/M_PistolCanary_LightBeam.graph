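# Unlit additive material for a glowing light beam effect, tinted by a glow color and intensity.
# Opacity combines bounds-based vertical gradients, Fresnel with adjustable vertical falloff, a contrast-adjustable height mask and a normal test that fades up- and down-facing surfaces.
# Camera-near and depth fades apply only on the deferred shading path.
# The graph also builds a UV-driven, vertical-axis billboard-style position offset with scale and camera offset.
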Material /PistolCanary_Countdown/Assets/FX/LightBeam/LightBeam_RudeEgg/M_PistolCanary_LightBeam
graph {
 properties {blendMode=BLEND_Additive, pixelDepthOffsetMode=PDOM_Legacy, shadingModel=MSM_Unlit, usedShadingModels=Unlit}
 nodes [
  n1 = ConstantBiasScale(None=TextureCoordinate()) {bias=-0.5, scale=1}
  n2 = MaterialFunctionCall(Float3=Subtract(A=ObjectPositionWS(), B=CameraPositionWS())) {materialFunction=&BreakOutFloat3Components}
  n3 = MaterialFunctionCall!(
   Vector1=Normalize(VectorInput=MaterialFunctionCall(X=@n2.R, Y=@n2.R) {materialFunction=&MakeFloat3}.Result)
   Vector2=Constant3Vector() {constant=<rgba 0 0 1 0>}
  ) {materialFunction=/Engine/Functions/Engine_MaterialFunctions02/Math/CreateThirdOrthogonalVector}
  n4 = Normalize(VectorInput=@n3.Vector3)
  n5 = ScalarParameter() {defaultValue=1, parameterName=Scale X}
  n6 = ComponentMask(None=@n1) {g=true}
  n7 = ScalarParameter() {defaultValue=1, parameterName=Scale Y}
  n8 = Add(A=Multiply(A=ComponentMask(None=@n1) {r=true}, B=Multiply(A=@n4, B=@n5)), B=Multiply(A=Multiply(A=@n3.Vector2, B=@n6), B=@n7))
  n9 = WorldPosition()
  Add(
   A=MaterialFunctionCall(Offset Amount=ScalarParameter() {parameterName=Camera Offset}) {materialFunction=/Engine/Functions/Engine_MaterialFunctions02/WorldPositionOffset/CameraOffset}.Camera Offset
   B=Add(A=Subtract(A=@n8, B=@n9.XYZ), B=ObjectPositionWS())
  )
  Constant() {r=1}
  n10 = ScalarParameter() {parameterName=NearCamFade, sortPriority=16}
  MaterialFunctionCall(Distance=@n10) {materialFunction=&Near_Camera_Fade}
  MaterialFunctionCall(X=@n5, Y=@n7) {materialFunction=&MakeFloat2}
  n11 = WorldPosition()
  n12 = ObjectPositionWS()
  n13 = Divide(
   A=Subtract(
    A=TransformPosition(None=@n11.XYZ) {transformSourceType=TRANSFORMPOSSOURCE_World}
    B=TransformPosition(None=@n12) {transformSourceType=TRANSFORMPOSSOURCE_World}
   )
   B=ObjectLocalBounds().Extents
  )
  n14 = ComponentMask(None=Saturate(None=Add(A=@n13) {constB=0.5})) {b=true}
  n15 = Multiply(
   A=Power(Base=OneMinus(None=@n14), Exp=ScalarParameter() {defaultValue=2, parameterName=Upper Edge Gradient Pwr, sortPriority=3})
   B=ScalarParameter() {defaultValue=1, parameterName=Upper Edge Gradient Mult, sortPriority=4}
  )
  n16 = Multiply(
   A=Power(Base=@n14, Exp=ScalarParameter() {defaultValue=1, parameterName=Lower Edge Gradient Pwr, sortPriority=5})
   B=ScalarParameter() {defaultValue=1, parameterName=Lower Edge Gradient Mult, sortPriority=6}
  )
  n17 = ScalarParameter() {defaultValue=1, parameterName=Edge Fresnel Brightness, sortPriority=8}
  n18 = ScalarParameter() {defaultValue=1, parameterName=Edge Fresnel Exp, sortPriority=9}
  n19 = MaterialFunctionCall(Float3=CameraVectorWS()) {materialFunction=&BreakOutFloat3Components}
  n20 = ComponentMask(None=VertexNormalWS()) {b=true}
  n21 = Normalize(VectorInput=LinearInterpolate(
    A=CameraVectorWS()
    B=MaterialFunctionCall(X=@n19.R, Y=@n19.R, Z=@n20) {materialFunction=&MakeFloat3}.Result
    Alpha=ScalarParameter() {parameterName=Fresnel Verticle Falloff}
   ))
  n22 = ScalarParameter() {defaultValue=3, parameterName=Opacity, sortPriority=15}
  n23 = Saturate(None=MaterialFunctionCall(Fade Length=@n10, Fade Offset=ScalarParameter() {parameterName=CameraNearCamFadeoffset}) {materialFunction=&CameraDepthFade}.Result)
  n24 = ShadingPathSwitch(
   Default=Constant() {r=1}
   Deferred=DepthFade(Opacity=LinearInterpolate(Alpha=@n23), FadeDistance=ScalarParameter() {defaultValue=100, parameterName=DepthFade, sortPriority=17})
  )
  n25 = Multiply(
   A=Multiply(A=Saturate(None=@n15), B=Saturate(None=@n16))
   B=Multiply(
    A=MaterialFunctionCall(Brightness=@n17, Exponent=@n18, Camera Vector=@n21) {materialFunction=&MF_Fresnel_FX}.Fresnel
    B=Multiply(A=@n22, B=@n24)
   )
  )
  n26 = MaterialFunctionCall() {materialFunction=&BoundingBoxBased_0-1_UVW}
  n27 = VectorParameter() {parameterName=Verical Mask Options, channelNames {a=contrast, b=add, g=mul, r=power}, defaultValue=<rgba 1 1 0 0>}
  n28 = Multiply(A=Power(Base=OneMinus(None=@n26.B), Exp=@n27.power), B=@n27.mul)
  n29 = Saturate(None=MaterialFunctionCall(In=Add(A=@n28, B=@n27.add), Contrast=@n27.contrast) {materialFunction=&CheapContrast}.Result)
  n30 = VertexNormalWS()
  n31 = Constant3Vector() {constant=<rgba 0 0 1 0>}
  n32 = OneMinus(None=Abs(None=DotProduct(A=@n30, B=@n31)))
  Multiply(A=Multiply(A=Saturate(None=@n25), B=Multiply(A=@n29, B=@n32)), B=ScalarParameter() {defaultValue=1, parameterName=Master Opacity})
  Reroute(None=Multiply(
    A=VectorParameter() {parameterName=GlowColor, sortPriority=1, defaultValue=<rgba 9 3.7221 0 3>}.RGB
    B=ScalarParameter() {defaultValue=1, parameterName=Emissive Intensity, sortPriority=2}
   ))
  ScalarParameter() {defaultValue=1, parameterName=Scale}
 ]
}
